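# Post-process-style AnimBP for the female medium Scrap jumpsuit body part, adding secondary motion to the player's outfit.
# It copies the parent mesh pose, simulates lapels, buckles, pouches and zippers with AnimDynamics, then applies two Control Rigs.
# That cached pose feeds three lower-leg dynamics variants, selected by bool blends.
# The EventGraph feeds foot and calf data to material WPO functions, enabling cloth wind during skydive floating.
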AnimBlueprint /Game/Characters/Player/Female/Medium/Bodies/F_MED_Jumpsuit_Scrap/Meshes/F_MED_Jumpsuit_Scrap_AnimBP
target_skeleton=&/F_MED_Jumpsuit_Scrap_Skeleton
graphs [
 AnimGraph {nodes [
   n1 = Anim Dynamics - Bone: lower_leg_right_back_dyn(ComponentPose=Local To Component(LocalPose=Use cached pose 'prerigid'()), Alpha=1, ExternalForce="0,0,0") {
    angularDampingOverride=0.9999
    angularSpringConstant=75
    bAngularSpring=true
    bGravityOverrideInSimSpace=true
    bOverrideAngularDamping=true
    simulationSpace=RootRelative
    alphaBoolBlend {blendOption=Linear}
    boundBone {boneName=lower_leg_right_back_dyn}
    componentLinearAccScale=<1 1 1>
    componentLinearVelScale=<1 1 1>
    physicsBodyDefinitions [{
      sphereCollisionRadius=0
      boundBone {boneName=lower_leg_right_back_dyn}
      boxExtents=<10 20 20>
      constraintSetup {angularLimitsMax=<1 -2 5>, angularLimitsMin=<-1 -20 -5>, angularTarget=<0 0 0>}
      localJointOffset=<1 0 -0.2>
     }]
    relativeSpaceBone {boneName=foot_r}
    sphericalLimits [{limitRadius=5, drivingBone {boneName=foot_r}, sphereLocalOffset=<0 0 0>}]
   }
   n2 = Anim Dynamics - Bone: lower_leg_right_front_dyn(ComponentPose=@n1, Alpha=1, ExternalForce="0,0,0") {
    angularDampingOverride=0.9999
    angularSpringConstant=75
    bAngularSpring=true
    bGravityOverrideInSimSpace=true
    bOverrideAngularDamping=true
    simulationSpace=RootRelative
    alphaBoolBlend {blendOption=Linear}
    boundBone {boneName=lower_leg_right_front_dyn}
    componentLinearAccScale=<1 1 1>
    componentLinearVelScale=<1 1 1>
    physicsBodyDefinitions [{
      sphereCollisionRadius=0
      boundBone {boneName=lower_leg_right_front_dyn}
      boxExtents=<10 20 20>
      constraintSetup {angularLimitsMax=<1 -15 1>, angularLimitsMin=<-1 -20 -1>, angularTarget=<0 0 0>}
      localJointOffset=<1 0 -0.2>
     }]
    sphericalLimits [{limitRadius=5, drivingBone {boneName=foot_r}, sphereLocalOffset=<0 0 0>}]
   }
   n3 = Anim Dynamics - Bone: lower_leg_left_back_dyn(ComponentPose=@n2, Alpha=1, ExternalForce="0,0,0") {
    angularDampingOverride=0.9999
    angularSpringConstant=75
    bAngularSpring=true
    bGravityOverrideInSimSpace=true
    bOverrideAngularDamping=true
    simulationSpace=RootRelative
    alphaBoolBlend {blendOption=Linear}
    boundBone {boneName=lower_leg_left_back_dyn}
    componentLinearAccScale=<1 1 1>
    componentLinearVelScale=<1 1 1>
    physicsBodyDefinitions [{
      sphereCollisionRadius=0
      boundBone {boneName=lower_leg_left_back_dyn}
      boxExtents=<10 20 20>
      constraintSetup {angularLimitsMax=<1 -2 5>, angularLimitsMin=<-1 -20 -5>, angularTarget=<0 0 0>}
      localJointOffset=<1 0 -0.2>
     }]
    sphericalLimits [{limitRadius=5, drivingBone {boneName=foot_l}, sphereLocalOffset=<0 0 0>}]
   }
   n4 = Anim Dynamics - Bone: leg_left_front_pants_dyn(ComponentPose=@n3, Alpha=1, ExternalForce="0,0,0") {
    angularDampingOverride=0.9999
    angularSpringConstant=75
    bAngularSpring=true
    bGravityOverrideInSimSpace=true
    bOverrideAngularDamping=true
    simulationSpace=RootRelative
    alphaBoolBlend {blendOption=Linear}
    boundBone {boneName=leg_left_front_pants_dyn}
    componentLinearAccScale=<1 1 1>
    componentLinearVelScale=<1 1 1>
    physicsBodyDefinitions [{
      sphereCollisionRadius=0
      boundBone {boneName=leg_left_front_pants_dyn}
      boxExtents=<10 20 20>
      constraintSetup {angularLimitsMax=<1 -5 1>, angularLimitsMin=<-1 -20 -1>, angularTarget=<0 0 0>}
      localJointOffset=<1 0 -0.2>
     }]
    sphericalLimits [{limitRadius=5, drivingBone {boneName=foot_l}, sphereLocalOffset=<0 0 0>}]
   }
   n5 = Component To Local(ComponentPose=@n4)
   n6 = "Constraint - knee_pad_mstr (RR)"!(ComponentPose=Local To Component(LocalPose=Use cached pose 'prerigid'()), Alpha=1, ConstraintWeights_0=0.4, ConstraintWeights_1:float=0.6) -> (Pose) {
    alphaBoolBlend {blendOption=Linear}
    boneToModify {boneName=knee_pad_mstr}
    constraintSetup [{transformType=Rotation, targetBone {boneName=thigh_l}}, {transformType=Rotation, targetBone {boneName=calf_l}}]
    constraintWeights [1,1]
   }
   n7 = "Transform (Modify) Bone - Bone: knee_pad_mstr"(ComponentPose=@n6, Alpha=1, Translation="2,0,0", Rotation="0,0,0", Scale="1,1,1") {
    rotationMode=BMM_Additive
    rotationSpace=BCS_ParentBoneSpace
    translationMode=BMM_Additive
    translationSpace=BCS_ParentBoneSpace
    alphaBoolBlend {blendOption=Linear}
    boneToModify {boneName=knee_pad_mstr}
    translation {x=2}
   }
   n8 = "Transform (Modify) Bone - Bone: lower_leg_right_front_mstr"(ComponentPose=@n7, Alpha=1, Translation="0,-0.75,0", Rotation="-10,0,0", Scale="1,1,1") {
    rotationMode=BMM_Additive
    rotationSpace=BCS_ParentBoneSpace
    translationMode=BMM_Additive
    translationSpace=BCS_ParentBoneSpace
    alphaBoolBlend {blendOption=Linear}
    boneToModify {boneName=lower_leg_right_front_mstr}
    rotation {pitch=-10}
    translation {y=-0.75}
   }
   n9 = "Transform (Modify) Bone - Bone: lower_leg_right_back_mstr"(ComponentPose=@n8, Alpha=1, Translation="0,-1.5,0", Rotation="10,25,0", Scale="1,1,1") {
    rotationMode=BMM_Additive
    rotationSpace=BCS_ParentBoneSpace
    translationMode=BMM_Additive
    translationSpace=BCS_ParentBoneSpace
    alphaBoolBlend {blendOption=Linear}
    boneToModify {boneName=lower_leg_right_back_mstr}
    rotation {pitch=10, yaw=25}
    translation {y=-1.5}
   }
   n10 = "Transform (Modify) Bone - Bone: leg_left_front_pants_dyn"(ComponentPose=@n9, Alpha=1, Translation="0,0,0", Rotation="0,3,0", Scale="1,1,1") {
    rotationMode=BMM_Additive
    rotationSpace=BCS_ParentBoneSpace
    translationMode=BMM_Additive
    translationSpace=BCS_ParentBoneSpace
    alphaBoolBlend {blendOption=Linear}
    boneToModify {boneName=leg_left_front_pants_dyn}
    rotation {yaw=3}
   }
   n11 = Anim Dynamics - Bone: lower_leg_right_back_dyn(ComponentPose=@n10, Alpha=1, ExternalForce="0,0,0") {
    angularDampingOverride=0.9999
    angularSpringConstant=75
    bAngularSpring=true
    bGravityOverrideInSimSpace=true
    bOverrideAngularDamping=true
    simulationSpace=RootRelative
    alphaBoolBlend {blendOption=Linear}
    boundBone {boneName=lower_leg_right_back_dyn}
    componentLinearAccScale=<1 1 1>
    componentLinearVelScale=<1 1 1>
    physicsBodyDefinitions [{
      sphereCollisionRadius=0
      boundBone {boneName=lower_leg_right_back_dyn}
      boxExtents=<10 20 20>
      constraintSetup {angularLimitsMax=<1 0 5>, angularLimitsMin=<-1 -20 -5>, angularTarget=<0 0 0>}
      localJointOffset=<1 0 -0.2>
     }]
   }
   n12 = Anim Dynamics - Bone: lower_leg_right_front_dyn(ComponentPose=@n11, Alpha=1, ExternalForce="0,0,0") {
    angularDampingOverride=0.9999
    angularSpringConstant=75
    bAngularSpring=true
    bGravityOverrideInSimSpace=true
    bOverrideAngularDamping=true
    simulationSpace=RootRelative
    alphaBoolBlend {blendOption=Linear}
    boundBone {boneName=lower_leg_right_front_dyn}
    componentLinearAccScale=<1 1 1>
    componentLinearVelScale=<1 1 1>
    physicsBodyDefinitions [{
      sphereCollisionRadius=0
      boundBone {boneName=lower_leg_right_front_dyn}
      boxExtents=<10 20 20>
      constraintSetup {angularLimitsMax=<1 5 10>, angularLimitsMin=<-1 -5 0>, angularTarget=<0 0 0>}
      localJointOffset=<1 0 -0.2>
     }]
   }
   n13 = Anim Dynamics - Bone: lower_leg_left_back_dyn(ComponentPose=@n12, Alpha=1, ExternalForce="0,0,0") {
    angularDampingOverride=0.9999
    angularSpringConstant=75
    bAngularSpring=true
    bGravityOverrideInSimSpace=true
    bOverrideAngularDamping=true
    simulationSpace=RootRelative
    alphaBoolBlend {blendOption=Linear}
    boundBone {boneName=lower_leg_left_back_dyn}
    componentLinearAccScale=<1 1 1>
    componentLinearVelScale=<1 1 1>
    physicsBodyDefinitions [{
      sphereCollisionRadius=0
      boundBone {boneName=lower_leg_left_back_dyn}
      boxExtents=<10 20 20>
      constraintSetup {angularLimitsMax=<1 -2 5>, angularLimitsMin=<-1 -20 -5>, angularTarget=<0 0 0>}
      localJointOffset=<1 0 -0.2>
     }]
   }
   n14 = Anim Dynamics - Bone: leg_left_front_pants_dyn(ComponentPose=@n13, Alpha=1, ExternalForce="0,0,0") {
    angularDampingOverride=0.9999
    angularSpringConstant=75
    bAngularSpring=true
    bGravityOverrideInSimSpace=true
    bOverrideAngularDamping=true
    simulationSpace=RootRelative
    alphaBoolBlend {blendOption=Linear}
    boundBone {boneName=leg_left_front_pants_dyn}
    componentLinearAccScale=<1 1 1>
    componentLinearVelScale=<1 1 1>
    physicsBodyDefinitions [{
      sphereCollisionRadius=0
      boundBone {boneName=leg_left_front_pants_dyn}
      boxExtents=<10 20 20>
      constraintSetup {angularLimitsMax=<1 -5 1>, angularLimitsMin=<-1 -20 -1>, angularTarget=<0 0 0>}
      localJointOffset=<1 0 -0.2>
     }]
   }
   n15 = Component To Local(ComponentPose=@n14)
   n16 = Anim Dynamics - Bone: lower_leg_right_back_dyn(ComponentPose=Local To Component(LocalPose=Use cached pose 'prerigid'()), Alpha=1, ExternalForce="0,0,0") {
    angularDampingOverride=0.9999
    angularSpringConstant=75
    bAngularSpring=true
    bGravityOverrideInSimSpace=true
    bOverrideAngularDamping=true
    simulationSpace=RootRelative
    alphaBoolBlend {blendOption=Linear}
    boundBone {boneName=lower_leg_right_back_dyn}
    componentLinearAccScale=<1 1 1>
    componentLinearVelScale=<1 1 1>
    physicsBodyDefinitions [{
      sphereCollisionRadius=0
      boundBone {boneName=lower_leg_right_back_dyn}
      boxExtents=<10 20 20>
      constraintSetup {angularLimitsMax=<1 -10 5>, angularLimitsMin=<-1 -30 -5>, angularTarget=<0 0 0>}
      localJointOffset=<1 0 -0.2>
     }]
   }
   n17 = Anim Dynamics - Bone: lower_leg_right_front_dyn(ComponentPose=@n16, Alpha=1, ExternalForce="0,0,0") {
    angularDampingOverride=0.9999
    angularSpringConstant=75
    bAngularSpring=true
    bGravityOverrideInSimSpace=true
    bOverrideAngularDamping=true
    simulationSpace=RootRelative
    alphaBoolBlend {blendOption=Linear}
    boundBone {boneName=lower_leg_right_front_dyn}
    componentLinearAccScale=<1 1 1>
    componentLinearVelScale=<1 1 1>
    physicsBodyDefinitions [{
      sphereCollisionRadius=0
      boundBone {boneName=lower_leg_right_front_dyn}
      boxExtents=<10 20 20>
      constraintSetup {angularLimitsMax=<1 -10 10>, angularLimitsMin=<-1 -30 0>, angularTarget=<0 0 0>}
      localJointOffset=<1 0 -0.2>
     }]
   }
   n18 = Anim Dynamics - Bone: lower_leg_left_back_dyn(ComponentPose=@n17, Alpha=1, ExternalForce="0,0,0") {
    angularDampingOverride=0.9999
    angularSpringConstant=75
    bAngularSpring=true
    bGravityOverrideInSimSpace=true
    bOverrideAngularDamping=true
    simulationSpace=RootRelative
    alphaBoolBlend {blendOption=Linear}
    boundBone {boneName=lower_leg_left_back_dyn}
    componentLinearAccScale=<1 1 1>
    componentLinearVelScale=<1 1 1>
    physicsBodyDefinitions [{
      sphereCollisionRadius=0
      boundBone {boneName=lower_leg_left_back_dyn}
      boxExtents=<10 20 20>
      constraintSetup {angularLimitsMax=<1 -10 5>, angularLimitsMin=<-1 -20 -5>, angularTarget=<0 0 0>}
      localJointOffset=<1 0 -0.2>
     }]
   }
   n19 = Anim Dynamics - Bone: leg_left_front_pants_dyn(ComponentPose=@n18, Alpha=1, ExternalForce="0,0,0") {
    angularDampingOverride=0.9999
    angularSpringConstant=75
    bAngularSpring=true
    bGravityOverrideInSimSpace=true
    bOverrideAngularDamping=true
    simulationSpace=RootRelative
    alphaBoolBlend {blendOption=Linear}
    boundBone {boneName=leg_left_front_pants_dyn}
    componentLinearAccScale=<1 1 1>
    componentLinearVelScale=<1 1 1>
    physicsBodyDefinitions [{
      sphereCollisionRadius=0
      boundBone {boneName=leg_left_front_pants_dyn}
      boxExtents=<10 20 20>
      constraintSetup {angularLimitsMax=<1 -2 0>, angularLimitsMin=<-1 -20 -10>, angularTarget=<0 0 0>}
      localJointOffset=<1 0 -0.2>
     }]
   }
   Output Pose(Result=Blend Poses by bool(BlendPose_0=Component To Local(ComponentPose=@n19), BlendPose_1=Blend Poses by bool(BlendPose_0=@n15, BlendPose_1=@n5) {blendType=Linear}) {blendType=Linear}) {layerGroup=None}
   n20 = "Transform (Modify) Bone - Bone: collar_b_mstr"(
    ComponentPose=Local To Component(LocalPose=Copy Pose From Mesh() {bUseAttachedParent=true})
    Alpha=1
    Translation="0,-0.5,0"
    Rotation="0,-5,0"
    Scale="1,1,1"
   ) {
    rotationMode=BMM_Additive
    rotationSpace=BCS_ParentBoneSpace
    translationMode=BMM_Additive
    translationSpace=BCS_ParentBoneSpace
    alphaBoolBlend {blendOption=Linear}
    boneToModify {boneName=collar_b_mstr}
    rotation {yaw=-5}
    translation {y=-0.5}
   }
   n21 = Anim Dynamics - Bone: lapel_l_01_dyn(ComponentPose=@n20, Alpha=1, ExternalForce="0,0,0") {
    angularDampingOverride=0.9999
    angularSpringConstant=75
    bAngularSpring=true
    bGravityOverrideInSimSpace=true
    bOverrideAngularDamping=true
    simulationSpace=RootRelative
    alphaBoolBlend {blendOption=Linear}
    boundBone {boneName=lapel_l_01_dyn}
    componentLinearAccScale=<1 1 1>
    componentLinearVelScale=<1 1 1>
    physicsBodyDefinitions [{
      sphereCollisionRadius=0
      boundBone {boneName=lapel_l_01_dyn}
      boxExtents=<10 20 20>
      constraintSetup {angularLimitsMax=<1 15 1>, angularLimitsMin=<-1 -45 -1>, angularTarget=<0 0 0>}
      localJointOffset=<1 0 -0.2>
     }]
   }
   n22 = Anim Dynamics - Bone: lapel_r_01_dyn(ComponentPose=@n21, Alpha=1, ExternalForce="0,0,0") {
    angularDampingOverride=0.9999
    angularSpringConstant=75
    bAngularSpring=true
    bGravityOverrideInSimSpace=true
    bOverrideAngularDamping=true
    simulationSpace=RootRelative
    alphaBoolBlend {blendOption=Linear}
    boundBone {boneName=lapel_r_01_dyn}
    componentLinearAccScale=<1 1 1>
    componentLinearVelScale=<1 1 1>
    physicsBodyDefinitions [{
      sphereCollisionRadius=0
      boundBone {boneName=lapel_r_01_dyn}
      boxExtents=<10 20 20>
      constraintSetup {angularLimitsMax=<1 15 1>, angularLimitsMin=<-1 -45 -1>, angularTarget=<0 0 0>}
      localJointOffset=<1 0 -0.2>
     }]
   }
   n23 = Anim Dynamics - Bone: chest_buckle_01_dyn(ComponentPose=@n22, Alpha=1, ExternalForce="0,0,0") {
    angularDampingOverride=0.9999
    angularSpringConstant=75
    bAngularSpring=true
    bGravityOverrideInSimSpace=true
    bOverrideAngularDamping=true
    simulationSpace=RootRelative
    alphaBoolBlend {blendOption=Linear}
    boundBone {boneName=chest_buckle_01_dyn}
    componentLinearAccScale=<1 1 1>
    componentLinearVelScale=<1 1 1>
    physicsBodyDefinitions [{
      sphereCollisionRadius=0
      boundBone {boneName=chest_buckle_01_dyn}
      boxExtents=<10 20 20>
      constraintSetup {angularLimitsMax=<1 0 1>, angularLimitsMin=<-1 -45 -1>, angularTarget=<0 0 0>}
      localJointOffset=<1 0 -0.2>
     }]
   }
   n24 = Anim Dynamics - Bone: chest_buckle_02_dyn(ComponentPose=@n23, Alpha=1, ExternalForce="0,0,0") {
    angularDampingOverride=0.9999
    angularSpringConstant=75
    bAngularSpring=true
    bGravityOverrideInSimSpace=true
    bOverrideAngularDamping=true
    simulationSpace=RootRelative
    alphaBoolBlend {blendOption=Linear}
    boundBone {boneName=chest_buckle_02_dyn}
    componentLinearAccScale=<1 1 1>
    componentLinearVelScale=<1 1 1>
    physicsBodyDefinitions [{
      sphereCollisionRadius=0
      boundBone {boneName=chest_buckle_02_dyn}
      boxExtents=<10 20 20>
      constraintSetup {angularLimitsMax=<1 0 1>, angularLimitsMin=<-1 -45 -1>, angularTarget=<0 0 0>}
      localJointOffset=<1 0 -0.2>
     }]
   }
   n25 = Anim Dynamics - Bone: chest_pouch_01_dyn(ComponentPose=@n24, Alpha=1, ExternalForce="0,0,0") {
    angularDampingOverride=0.9999
    angularSpringConstant=10
    bAngularSpring=true
    bGravityOverrideInSimSpace=true
    bOverrideAngularDamping=true
    simulationSpace=RootRelative
    alphaBoolBlend {blendOption=Linear}
    boundBone {boneName=chest_pouch_01_dyn}
    componentAppliedLinearAccClamp=<5000 5000 5000>
    componentLinearAccScale=<0.5 0.5 0.5>
    componentLinearVelScale=<0.5 0.5 0.5>
    physicsBodyDefinitions [{
      sphereCollisionRadius=0
      boundBone {boneName=chest_pouch_01_dyn}
      boxExtents=<10 20 20>
      constraintSetup {angularLimitsMax=<1 15 2.5>, angularLimitsMin=<-1 -15 -2.5>, angularTarget=<0 0 0>}
      localJointOffset=<1 0 -0.2>
     }]
   }
   n26 = Anim Dynamics - Bone: chest_pouich_02_dyn(ComponentPose=@n25, Alpha=1, ExternalForce="0,0,0") {
    angularDampingOverride=0.9999
    angularSpringConstant=10
    bAngularSpring=true
    bGravityOverrideInSimSpace=true
    bOverrideAngularDamping=true
    simulationSpace=RootRelative
    alphaBoolBlend {blendOption=Linear}
    boundBone {boneName=chest_pouich_02_dyn}
    componentAppliedLinearAccClamp=<5000 5000 5000>
    componentLinearAccScale=<0.5 0.5 0.5>
    componentLinearVelScale=<0.5 0.5 0.5>
    physicsBodyDefinitions [{
      sphereCollisionRadius=0
      boundBone {boneName=chest_pouich_02_dyn}
      boxExtents=<10 20 20>
      constraintSetup {angularLimitsMax=<1 15 2.5>, angularLimitsMin=<-1 -15 -2.5>, angularTarget=<0 0 0>}
      localJointOffset=<1 0 -0.2>
     }]
   }
   n27 = "Constraint - zipper_leg_r_mstr (PPPPPPP)"!(
    ComponentPose=@n26
    Alpha=1
    ConstraintWeights_0=0.1135
    ConstraintWeights_1:float=0.0009
    ConstraintWeights_2:float=0.5755
    ConstraintWeights_3:float=0.1893
    ConstraintWeights_4:float=0
    ConstraintWeights_5:float=0.1185
    ConstraintWeights_6:float=0.0024
   ) -> (Pose) {
    alphaBoolBlend {blendOption=Linear}
    boneToModify {boneName=zipper_leg_r_mstr}
    constraintSetup [
     {targetBone {boneName=pelvis}}
     {targetBone {boneName=thigh_twist_01_l}}
     {targetBone {boneName=thigh_twist_01_r}}
     {targetBone {boneName=thigh_r}}
     {targetBone {boneName=deform_glute_l}}
     {targetBone {boneName=deform_glute_r}}
     {targetBone {boneName=deform_groin_r}}
    ]
    constraintWeights [1,1,1,1,1,1,1]
   }
   n28 = Anim Dynamics - Bone: zipper_leg_r_dyn(ComponentPose=@n27, Alpha=1, ExternalForce="0,0,0") {
    angularDampingOverride=0.9999
    bOverrideAngularDamping=true
    bUseGravityOverride=true
    bUsePlanarLimit=false
    simulationSpace=RootRelative
    alphaBoolBlend {blendOption=Linear}
    boundBone {boneName=zipper_leg_r_dyn}
    componentAppliedLinearAccClamp=<3000 3000 3000>
    componentLinearAccScale=<0.5 0.5 0.5>
    componentLinearVelScale=<0.25 0.25 0.25>
    gravityOverride {z=-980}
    physicsBodyDefinitions [{
      boundBone {boneName=zipper_leg_r_dyn}
      boxExtents=<2 10 10>
      constraintSetup {angularLimitsMax=<2 5 30>, angularLimitsMin=<-2 -30 -30>, angularTarget=<0 0 0>}
      localJointOffset=<3 0 0>
     }]
   }
   n29 = Anim Dynamics - Bone: leg_pouch_l_dyn(ComponentPose=@n28, Alpha=1, ExternalForce="0,0,0") {
    angularDampingOverride=0.9999
    angularSpringConstant=10
    bAngularSpring=true
    bGravityOverrideInSimSpace=true
    bOverrideAngularDamping=true
    simulationSpace=RootRelative
    alphaBoolBlend {blendOption=Linear}
    boundBone {boneName=leg_pouch_l_dyn}
    componentAppliedLinearAccClamp=<5000 5000 5000>
    componentLinearAccScale=<0.5 0.5 0.5>
    componentLinearVelScale=<0.5 0.5 0.5>
    physicsBodyDefinitions [{
      sphereCollisionRadius=0
      boundBone {boneName=leg_pouch_l_dyn}
      boxExtents=<10 20 20>
      constraintSetup {angularLimitsMax=<1 2.5 2.5>, angularLimitsMin=<-1 -2.5 -2.5>, angularTarget=<0 0 0>}
      localJointOffset=<1 0 -0.2>
     }]
   }
   n30 = Anim Dynamics - Bone: leg_pouch_l_strap_dyn(ComponentPose=@n29, Alpha=1, ExternalForce="0,0,0") {
    angularDampingOverride=0.9999
    angularSpringConstant=75
    bAngularSpring=true
    bGravityOverrideInSimSpace=true
    bOverrideAngularDamping=true
    simulationSpace=RootRelative
    alphaBoolBlend {blendOption=Linear}
    boundBone {boneName=leg_pouch_l_strap_dyn}
    componentLinearAccScale=<1 1 1>
    componentLinearVelScale=<1 1 1>
    physicsBodyDefinitions [{
      sphereCollisionRadius=0
      boundBone {boneName=leg_pouch_l_strap_dyn}
      boxExtents=<10 20 20>
      constraintSetup {angularLimitsMax=<1 0 1>, angularLimitsMin=<-1 -15 -1>, angularTarget=<0 0 0>}
      localJointOffset=<1 0 -0.2>
     }]
   }
   n31 = Anim Dynamics - Bone: leg_pouch_r_01_dyn(ComponentPose=@n30, Alpha=1, ExternalForce="0,0,0") {
    angularDampingOverride=0.9999
    angularSpringConstant=10
    bAngularSpring=true
    bGravityOverrideInSimSpace=true
    bOverrideAngularDamping=true
    simulationSpace=RootRelative
    alphaBoolBlend {blendOption=Linear}
    boundBone {boneName=leg_pouch_r_01_dyn}
    componentAppliedLinearAccClamp=<5000 5000 5000>
    componentLinearAccScale=<0.5 0.5 0.5>
    componentLinearVelScale=<0.5 0.5 0.5>
    physicsBodyDefinitions [{
      sphereCollisionRadius=0
      boundBone {boneName=leg_pouch_r_01_dyn}
      boxExtents=<10 20 20>
      constraintSetup {angularLimitsMax=<1 2.5 2.5>, angularLimitsMin=<-1 -2.5 -2.5>, angularTarget=<0 0 0>}
      localJointOffset=<1 0 -0.2>
     }]
   }
   n32 = Anim Dynamics - Bone: leg_pouch_r_02_dyn(ComponentPose=@n31, Alpha=1, ExternalForce="0,0,0") {
    angularDampingOverride=0.9999
    angularSpringConstant=10
    bAngularSpring=true
    bGravityOverrideInSimSpace=true
    bOverrideAngularDamping=true
    simulationSpace=RootRelative
    alphaBoolBlend {blendOption=Linear}
    boundBone {boneName=leg_pouch_r_02_dyn}
    componentAppliedLinearAccClamp=<5000 5000 5000>
    componentLinearAccScale=<0.5 0.5 0.5>
    componentLinearVelScale=<0.5 0.5 0.5>
    physicsBodyDefinitions [{
      sphereCollisionRadius=0
      boundBone {boneName=leg_pouch_r_02_dyn}
      boxExtents=<10 20 20>
      constraintSetup {angularLimitsMax=<1 2.5 2.5>, angularLimitsMin=<-1 -2.5 -2.5>, angularTarget=<0 0 0>}
      localJointOffset=<1 0 -0.2>
     }]
   }
   n33 = Anim Dynamics - Bone: lower_leg_pouch_r_dyn(ComponentPose=@n32, Alpha=1, ExternalForce="0,0,0") {
    angularDampingOverride=0.9999
    angularSpringConstant=10
    bAngularSpring=true
    bGravityOverrideInSimSpace=true
    bOverrideAngularDamping=true
    simulationSpace=RootRelative
    alphaBoolBlend {blendOption=Linear}
    boundBone {boneName=lower_leg_pouch_r_dyn}
    componentLinearAccScale=<1 1 1>
    componentLinearVelScale=<1 1 1>
    physicsBodyDefinitions [{
      sphereCollisionRadius=0
      boundBone {boneName=lower_leg_pouch_r_dyn}
      boxExtents=<10 20 20>
      constraintSetup {angularLimitsMax=<1 5 5>, angularLimitsMin=<-1 -15 -5>, angularTarget=<0 0 0>}
      localJointOffset=<1 0 -0.2>
     }]
   }
   n34 = Anim Dynamics - Bone: lower_leg_pouch_zipper_r_dyn(ComponentPose=@n33, Alpha=1, ExternalForce="0,0,0") {
    angularDampingOverride=0.9999
    angularSpringConstant=75
    bAngularSpring=true
    bGravityOverrideInSimSpace=true
    bOverrideAngularDamping=true
    simulationSpace=RootRelative
    alphaBoolBlend {blendOption=Linear}
    boundBone {boneName=lower_leg_pouch_zipper_r_dyn}
    componentLinearAccScale=<1 1 1>
    componentLinearVelScale=<1 1 1>
    physicsBodyDefinitions [{
      sphereCollisionRadius=0
      boundBone {boneName=lower_leg_pouch_zipper_r_dyn}
      boxExtents=<10 20 20>
      constraintSetup {angularLimitsMax=<1 45 45>, angularLimitsMin=<-1 -45 -45>, angularTarget=<0 0 0>}
      localJointOffset=<1 0 -0.2>
     }]
   }
   n35 = Control Rig(Source=Component To Local(ComponentPose=@n34)) {
    alphaBoolBlend {blendOption=Linear}
    controlRigAssetReference {blueprintRigClass=&/F_MED_JumpSuit_Scrap_DeformRig_Master_ControlRig.F_MED_JumpSuit_Scrap_DeformRig_Master_ControlRig_C}
   }
   n36 = Control Rig!(Source=@n35, Alpha, Crouch_CTRL:bool=Get bIsCrouching(), SkyDive_CTRL:bool=Get bIsSkyDiving()) -> (Pose) {
    alphaBoolBlend {blendOption=Linear}
    controlRigAssetReference {blueprintRigClass=&/F_MED_JumpSuit_Scrap_CtrlRig.F_MED_JumpSuit_Scrap_CtrlRig_C}
   }
   Save cached pose 'prerigid'(Pose=@n36) {cacheName=prerigid}
  ]}
 EventGraph {nodes [
   n1 = Event BlueprintUpdateAnimation()
   > Sequence!(execute) -> (then_0, then_1, then_2:Exec)
   > Set WPO Globals Function(
    Material=Get Mat Instance Dynamic()
    Skeletal Mesh=Get Skel Mesh()
    "Local Noise Transform  Location"=Get Local Noise Transform Location() -> (Local Noise Transform Location:Vector)
    Local Noise Offset=Get Local Noise Offset()
   )
   > Set WPO Regional Function!(
    execute
    Material:Material Instance Dynamic Object Reference=Get Mat Instance Dynamic()
    Overlay Material:Material Instance Dynamic Object Reference
    Skeletal Mesh:Skeletal Mesh Component Object Reference=Get Skel Mesh()
    Velocity Transform:Name=foot_r
    Velocity Interp Speed:double=Get Velocity Interp Speed()
    Collision Transform:Name=calf_r
    Collision Transform Axis:EAxis Enum
    Invert Collision Transform Axis:bool=false
    Delta Time X:double=@n1.DeltaTimeX
    Previous Velocity Transform Location:"Vector (by ref)"=Get R1 Velocity Transform Location()
    Previous Velocity Transform Velocity:"Vector (by ref)"=Get R1 Velocity Transform Velocity()
    Region:int=1
   ) -> (then)
   > Set WPO Regional Function(
    Material=Get Mat Instance Dynamic()
    Skeletal Mesh=Get Skel Mesh()
    Velocity Transform=foot_l
    Velocity Interp Speed=Get Velocity Interp Speed()
    Collision Transform=calf_l
    Invert Collision Transform Axis=true
    Delta Time X=@n1.DeltaTimeX
    Previous Velocity Transform Location=Get R2 Velocity Transform Location()
    Previous Velocity Transform Velocity=Get R2 Velocity Transform Velocity()
    Region=2
   )
   > n2 = Branch(Condition=Get bIsSkydiveFloating())
   > n3 = Set WPO Pho Wind Intensity(WPO Pho Wind Intensity=1)
   SetScalarParameterValue(execute=[@n4.then, @n3.then], self=Get Mat Instance Dynamic(), ParameterName=ClothSim_WindStrength, Value=Get WPO Pho Wind Intensity())
   n4 = Set WPO Pho Wind Intensity(execute=@n2.else, WPO Pho Wind Intensity=0)
   Event BlueprintInitializeAnimation()
   > Set Skel Mesh(Skel Mesh=GetOwnerSkeletalMeshComponent())
   n5 = TryGetPawnOwner()
   > Is Valid(InputObject=@n5)
   > n6 = Cast To FortPlayerPawnAthena(Object=@n5)
   > Set Player Pawn(Player Pawn:Fort Player Pawn Athena Object Reference=@n6) -> (Output_Get:Fort Player Pawn Athena Object Reference)
   Event OnCharacterCustomizationLoaded()
   > n7 = Set Skel Mesh(Skel Mesh=GetPartSkeletalMeshComponent())
   > n8 = Get Dynamic Material Instance(SkeletalMeshComponent=@n7)
   > Set Mat Instance Dynamic(Mat Instance Dynamic=@n8)
  ]}
]
variables [
 Local Noise Offset {type=Vector}
 Local Noise Transform Location {type=Vector}
 Mat Instance Dynamic {type=MaterialInstanceDynamic}
 Player Pawn {type=FortPlayerPawnAthena}
 R1 Velocity Transform Location {type=Vector}
 R1 Velocity Transform Velocity {type=Vector}
 R2 Velocity Transform Location {type=Vector}
 R2 Velocity Transform Velocity {type=Vector}
 Skel Mesh {type=SkeletalMeshComponent}
 Velocity Interp Speed {type=number, default=30}
 WPO Pho Wind Intensity {type=number}
]
