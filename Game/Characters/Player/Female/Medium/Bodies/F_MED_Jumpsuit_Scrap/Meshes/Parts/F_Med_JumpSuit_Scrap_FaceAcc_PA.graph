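# Physics asset for the female medium Scrap jumpsuit face accessory, defining secondary-motion bodies for bear, bunny and horn ear bones on the head.
# A head sphere anchors left and right master spheres through angularly free constraints.
# Capsule dynamic bones hang off each master with narrow swing and twist limits, tightest on the horns.
# Bunny ears chain two capsules per side, while bear and horn ears use one.
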PhysicsAsset /Game/Characters/Player/Female/Medium/Bodies/F_MED_Jumpsuit_Scrap/Meshes/Parts/F_Med_JumpSuit_Scrap_FaceAcc_PA
bodies {
 bear_ear_l_01_dyn [{shapeType=Capsule, length=2, radius=3, shapeName=bear_ear_l_01_dyn_capsule, center=<5 0 0>, rotation=<rot 90 0 0>}]
 bear_ear_l_mstr [{shapeType=Sphere, radius=0.505, shapeName=bear_ear_l_mstr_sphere}]
 bear_ear_r_01_dyn [{shapeType=Capsule, length=2, radius=3, shapeName=bear_ear_r_01_dyn_capsule, center=<5 0 0>, rotation=<rot 90 0 0>}]
 bear_ear_r_mstr [{shapeType=Sphere, radius=0.505, shapeName=bear_ear_r_mstr_sphere}]
 bunny_ear_l_01_dyn [{shapeType=Capsule, length=2, radius=3, shapeName=bunny_ear_l_01_dyn_capsule, center=<5 0 0>, rotation=<rot 90 0 0>}]
 bunny_ear_l_02_dyn [{shapeType=Capsule, length=3, radius=3, shapeName=bunny_ear_l_02_dyn_capsule, center=<5.75 0 0>, rotation=<rot 90 0 0>}]
 bunny_ear_l_mstr [{shapeType=Sphere, radius=0.505, shapeName=bunny_ear_l_mstr_sphere}]
 bunny_ear_r_01_dyn [{shapeType=Capsule, length=2, radius=3, shapeName=bunny_ear_r_01_dyn_capsule, center=<5 0 0>, rotation=<rot 90 0 0>}]
 bunny_ear_r_02_dyn [{shapeType=Capsule, length=3, radius=3, shapeName=bunny_ear_r_02_dyn_capsule, center=<5.75 0 0>, rotation=<rot 90 0 0>}]
 bunny_ear_r_mstr [{shapeType=Sphere, radius=0.505, shapeName=bunny_ear_r_mstr_sphere}]
 head [{shapeType=Sphere, radius=10, shapeName=head_sphere, center=<3.8003 0 0>}]
 horn_ear_l_01_dyn [{shapeType=Capsule, length=3, radius=2, shapeName=horn_ear_l_01_dyn_capsule, center=<3 0 0>, rotation=<rot 90 0 0>}]
 horn_ear_l_01_mstr [{shapeType=Sphere, radius=3.3196, shapeName=horn_ear_l_01_mstr_sphere, center=<-0.18 0.7607 0.4118>}]
 horn_ear_r_01_dyn [{shapeType=Capsule, length=3, radius=2, shapeName=horn_ear_r_01_dyn_capsule, center=<3 0 0>, rotation=<rot 90 0 0>}]
 horn_ear_r_01_mstr [{shapeType=Sphere, radius=3.3162, shapeName=horn_ear_r_01_mstr_sphere, center=<-0.2225 -0.7932 0.2966>}]
}
constraints [
 {bone1Name=bear_ear_l_01_dyn, bone2Name=bear_ear_l_mstr, swing1LimitDegrees=15, swing2LimitDegrees=15, twistLimitDegrees=5}
 {bone1Name=bear_ear_r_01_dyn, bone2Name=bear_ear_r_mstr, swing1LimitDegrees=15, swing2LimitDegrees=15, twistLimitDegrees=5}
 {bone1Name=bunny_ear_l_01_dyn, bone2Name=bunny_ear_l_mstr, swing1LimitDegrees=15, swing2LimitDegrees=15, twistLimitDegrees=5}
 {bone1Name=bunny_ear_l_02_dyn, bone2Name=bunny_ear_l_01_dyn, swing1LimitDegrees=15, swing2LimitDegrees=15, twistLimitDegrees=5}
 {bone1Name=bunny_ear_r_01_dyn, bone2Name=bunny_ear_r_mstr, swing1LimitDegrees=15, swing2LimitDegrees=15, twistLimitDegrees=5}
 {bone1Name=bunny_ear_r_02_dyn, bone2Name=bunny_ear_r_01_dyn, swing1LimitDegrees=15, swing2LimitDegrees=15, twistLimitDegrees=5}
 {
  bone1Name=bunny_ear_r_mstr
  bone2Name=head
  swing1LimitDegrees=45
  swing1Motion=Free
  swing2LimitDegrees=45
  swing2Motion=Free
  twistLimitDegrees=45
  twistMotion=Free
 }
 {
  bone1Name=bunny_ear_l_mstr
  bone2Name=head
  swing1LimitDegrees=45
  swing1Motion=Free
  swing2LimitDegrees=45
  swing2Motion=Free
  twistLimitDegrees=45
  twistMotion=Free
 }
 {
  bone1Name=bear_ear_r_mstr
  bone2Name=head
  swing1LimitDegrees=45
  swing1Motion=Free
  swing2LimitDegrees=45
  swing2Motion=Free
  twistLimitDegrees=45
  twistMotion=Free
 }
 {
  bone1Name=bear_ear_l_mstr
  bone2Name=head
  swing1LimitDegrees=45
  swing1Motion=Free
  swing2LimitDegrees=45
  swing2Motion=Free
  twistLimitDegrees=45
  twistMotion=Free
 }
 {
  bone1Name=horn_ear_l_01_mstr
  bone2Name=head
  swing1LimitDegrees=45
  swing1Motion=Free
  swing2LimitDegrees=45
  swing2Motion=Free
  twistLimitDegrees=45
  twistMotion=Free
 }
 {
  bone1Name=horn_ear_r_01_mstr
  bone2Name=head
  swing1LimitDegrees=45
  swing1Motion=Free
  swing2LimitDegrees=45
  swing2Motion=Free
  twistLimitDegrees=45
  twistMotion=Free
 }
 {bone1Name=horn_ear_l_01_dyn, bone2Name=horn_ear_l_01_mstr, swing1LimitDegrees=5, swing2LimitDegrees=5, twistLimitDegrees=5}
 {bone1Name=horn_ear_r_01_dyn, bone2Name=horn_ear_r_01_mstr, swing1LimitDegrees=5, swing2LimitDegrees=5, twistLimitDegrees=5}
]
